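# BR_Audio_Ground_Material is a Default Lit material, by name and path the ground surface for BR_Audio QA art.
# Its graph holds a single texture sample parameter, BR_Audio_Ground_Texture, which exposes the texture for override in material instances.
# The graph shows no other nodes and no link from that sample to a material output.
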Material /Game/QA/Art_QA/BR_Audio/Art/BR_Audio_Ground_Material
graph {
 properties {pixelDepthOffsetMode=PDOM_Legacy, usedShadingModels=Default Lit}
 nodes [TextureSampleParameter2D() {parameterName=BR_Audio_Ground_Texture}]
}
