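# BR_Audio_Orange_Gradient is an unlit, fully rough material in the BR_Audio QA art folder that shows a dimmed yellow-orange gradient texture.
# It samples the RGB of a texture borrowed from the Skyfire orb gameplay assets and multiplies it by a constant to darken it.
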Material /Game/QA/Art_QA/BR_Audio/Art/BR_Audio_Orange_Gradient
graph {
 properties {bFullyRough=true, pixelDepthOffsetMode=PDOM_Legacy, shadingModel=MSM_Unlit, usedShadingModels=Unlit}
 nodes [Multiply(A=TextureSample() {texture=/Skyfire/Gameplay/Orbs/14DoS_GIFBG_Yellow-Orange1}.RGB, B=Constant() {r=0.4})]
}
